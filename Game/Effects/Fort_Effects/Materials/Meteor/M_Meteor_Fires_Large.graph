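# Translucent unlit particle material for large meteor fire, with scrolling flames that also drive a distortion along vertex normals, weighted by vertex color.
# Three panned noise samples, offset by a dynamic parameter, form a contrast mask blending two HDR colors tinted by particle color.
# Opacity combines that mask with a gradient or dynamic wipe, Fresnel, depth and camera fades, with cheaper paths by shading path and quality.
# An optional portal fade clips it against a plane.
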Material /Game/Effects/Fort_Effects/Materials/Meteor/M_Meteor_Fires_Large
graph {
 properties {
  allowTranslucentCustomDepthWrites=true
  blendMode=BLEND_Translucent
  opacityMaskClipValue=0.01
  pixelDepthOffsetMode=PDOM_Legacy
  shadingModel=MSM_Unlit
  usedShadingModels=Unlit
 }
 nodes [
  n1 = Multiply(A=VertexNormalWS(), B=OneMinus(None=VertexColor().R))
  n2 = ComponentMask(None=VectorParameter() {group=UVScale, parameterName=UVScale, defaultValue=<rgba 1 1 0 0>}.RGB) {g=true, r=true}
  n3 = Multiply(A=TextureCoordinate() {vTiling=0.5}, B=@n2)
  n4 = DynamicParameter() {paramNames [UVOffset,Wipe,Param3,Param4]}
  n5 = AppendVector(A=ScalarParameter() {parameterName=Fire Speed U}, B=ScalarParameter() {defaultValue=0.001, parameterName=Fire Speed V})
  n6 = TextureSample(UVs=Panner(Coordinate=Add(A=@n3, B=@n4.UVOffset), Speed=@n5) {bFractionalPart=true, speedX=-0.05, speedY=1}) {texture=&T_NoisePack_01}
  n7 = Constant() {r=0.2}
  n8 = TextureCoordinate() {uTiling=4}
  n9 = Panner(Coordinate=Add(A=Multiply(A=@n2, B=@n8), B=@n4.UVOffset), Speed=@n5) {bFractionalPart=true, speedX=0.1, speedY=1}
  n10 = Add(A=@n4.UVOffset, B=Multiply(A=@n2, B=TextureCoordinate() {uTiling=12, vTiling=3}))
  n11 = Multiply(
   A=Add(
    A=MaterialFunctionCall(In=@n6.G, Contrast=@n7) {materialFunction=&CheapContrast}.Result
    B=TextureSample(UVs=@n9) {texture=&T_NoisePack_01}.G
   )
   B=TextureSample(UVs=Panner(Coordinate=@n10, Speed=@n5) {bFractionalPart=true, speedX=-0.1, speedY=1}) {texture=&T_NoisePack_01}.G
  )
  n12 = ScalarParameter() {defaultValue=25, parameterName=Distort}
  n13 = Panner(Coordinate=Add(A=@n4.UVOffset, B=TextureCoordinate()), Speed=@n5) {bFractionalPart=true, speedY=0.25}
  Add(
   A=Multiply(A=Multiply(A=@n1, B=@n11), B=@n12)
   B=Multiply(A=TextureSample(UVs=@n13) {texture=/Game/Global/Textures/Default/Masks_and_Gradients/CylGradient}.RGB, B=@n12) {constB=35}
  )
  n14 = TextureCoordinate()
  n15 = ComponentMask(None=@n14) {r=true}
  n16 = Multiply(A=@n15, B=OneMinus(None=@n15))
  n17 = MaterialFunctionCall(Float2=TextureCoordinate()) {materialFunction=&BreakOutFloat2Components}
  n18 = DynamicParameter() {paramNames [UVOffset,Wipe,Param3,Param4]}
  n19 = AppendVector(A=@n17.R, B=Add(A=@n17.G, B=LinearInterpolate(Alpha=@n18.Wipe) {constA=-1}))
  n20 = Multiply(A=Power(Base=TextureSample(UVs=@n19) {samplerSource=SSM_Clamp_WorldGroupSettings, texture=&CylGradient}.G)) {constB=20}
  n21 = Multiply(
   A=Power(Base=ComponentMask(None=@n14) {g=true}, Exp=ScalarParameter() {defaultValue=2, parameterName=SwipeMaskPower})
   B=Multiply(A=Multiply(A=@n16) {constB=4}, B=Clamp(None=@n20))
  )
  n22 = Multiply(A=@n21, B=@n11)
  n23 = MaterialFunctionCall() {materialFunction=&LinearGradient}
  n24 = ScalarParameter() {defaultValue=15, parameterName=FireLine}
  n25 = Power(Base=Add(A=@n11, B=Power(Base=@n23.VGradient, Exp=@n24))) {constExponent=0.5}
  n26 = Clamp(None=MaterialFunctionCall(
    In=StaticSwitchParameter(True=@n22, False=@n25) {parameterName=UseSwipe}
    Contrast=ScalarParameter() {defaultValue=0.5, parameterName=HardEdgeness}
   ) {materialFunction=&CheapContrast}.Result)
  n27 = LinearInterpolate(
   A=VectorParameter() {parameterName=Color_A, defaultValue=<rgba 10 0 0.1768 1>}.RGB
   B=VectorParameter() {parameterName=Color_B, defaultValue=<rgba 10 1.9576 0 1>}.RGB
   Alpha=@n26
  )
  n28 = Clamp(None=@n27)
  n29 = ParticleColor()
  n30 = Multiply(A=@n27, B=@n29.RGB)
  ShadingPathSwitch(
   Default=@n30
   Deferred=QualitySwitch(
    Default=MaterialFunctionCall(LDR Input=@n28, HDR Tint=@n30) {materialFunction=&DeriveHDRfromLDR}.RemappedRange
    Low=@n30
   )
  )
  n31 = StaticSwitchParameter(
   True=@n21
   False=Power(
    Base=MaterialFunctionCall() {materialFunction=&LinearGradient}.VGradient
    Exp=ScalarParameter() {defaultValue=8, parameterName=GradientContrast}
   )
  ) {parameterName=UseSwipe}
  n32 = Constant3Vector() {constant=<rgba 0 0 1 0>}
  n33 = ScalarParameter() {defaultValue=3, parameterName=FresnelExp}
  n34 = Fresnel(ExponentIn=@n33)
  n35 = Multiply(
   A=@n31
   B=StaticSwitchParameter(
    True=MaterialFunctionCall(Axis=@n32, Exponent=@n33) {materialFunction=&AxisAlignedFresnel}.Fresnel
    False=OneMinus(None=@n34)
   ) {defaultValue=true, parameterName=UseAxisAlignFresnel}
  )
  n36 = Multiply(A=@n26, B=ShadingPathSwitch(Default=@n31, Deferred=QualitySwitch(Default=@n35, Low=@n31)))
  n37 = DepthFade(Opacity=@n36, FadeDistance=ScalarParameter() {defaultValue=64, parameterName=FadeDistance}) {fadeDistanceDefault=50}
  n38 = Multiply(
   A=Multiply(A=@n29.A, B=ShadingPathSwitch(Default=@n36, Deferred=@n37))
   B=MaterialFunctionCall(
    Fade Length=ScalarParameter() {defaultValue=2500, parameterName=Fade Length}
    Fade Offset=ScalarParameter() {defaultValue=1250, parameterName=Fade Offset}
   ) {materialFunction=&CameraDepthFade}.Result
  )
  n39 = Subtract(A=VectorParameter() {parameterName=PlaneLocation, defaultValue=<rgba 24056 2399 4844 0>}.RGB, B=WorldPosition().XYZ)
  n40 = Saturate(None=DotProduct(A=Normalize(VectorInput=@n39), B=VectorParameter() {parameterName=MeteorDirection, defaultValue=<rgba 0.7755 -0.4565 -0.4361 0>}.RGB))
  StaticSwitchParameter(True=Multiply(A=@n38, B=Ceil(None=@n40)), False=@n38) {parameterName=PortalFade}
 ]
}
