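# Translucent unlit material for a glitched meteor effect, with vertex displacement, chromatic glow and dithered refraction.
# A world-space glitch mask and distance fade drive displacement along vertex normals and a push along MeteorDirection.
# View-space normals offset two tinted mask samples in opposite directions to fake chromatic aberration.
# Scene-normal Fresnel, near-camera fade, depth fade and an optional plane-based portal mask shape opacity, with simplified mobile paths.
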Material /Game/Effects/Fort_Effects/Materials/Meteor/M_Meteor_Glitched
graph {
 properties {
  bUsedWithInstancedStaticMeshes=true
  bUsedWithNanite=true
  blendMode=BLEND_Translucent
  pixelDepthOffsetMode=PDOM_Legacy
  refractionMethod=RM_IndexOfRefraction
  shadingModel=MSM_Unlit
  usedShadingModels=Unlit
 }
 nodes [
  n1 = MaterialFunctionCall!(MaskScaleMult) {materialFunction=/Game/Athena/Effects/Materials/Meteor/MF_Glitch_FX_World_Mask}
  n2 = LinearInterpolate(Alpha=@n1.DistanceFadeMask) {constA=10, constB=300}
  n3 = Multiply(A=@n1.WorldSpaceMask, B=LinearInterpolate(Alpha=@n1.DistanceFadeMask) {constA=0.15, constB=2})
  n4 = VectorParameter() {parameterName=MeteorDirection, defaultValue=<rgba 0.7755 -0.4565 -0.4361 0>}
  n5 = Multiply(A=@n4.RGB, B=DotProduct(A=@n4.RGB, B=VertexNormalWS()))
  Add(
   A=Multiply(A=LinearInterpolate(B=@n2, Alpha=@n1.WorldSpaceMask) {constA=-10, constB=80}, B=VertexNormalWS())
   B=Multiply(A=Multiply(A=@n3, B=@n5), B=ScalarParameter() {defaultValue=500, parameterName=Push})
  )
  n6 = TextureCoordinate()
  n7 = ComponentMask(None=Transform(None=TextureSample() {samplerType=SAMPLERTYPE_Normal, texture=/Game/Athena/Environments/Meteor/Textures/Meteor_N}.RGB) {transformType=TRANSFORM_View}) {g=true, r=true}
  n8 = ScalarParameter() {defaultValue=0.001, parameterName=ChromaOffset}
  n9 = Multiply(A=@n8) {constB=-1}
  n10 = TextureSample(UVs=Add(A=@n6, B=Multiply(A=@n7, B=@n9))) {constMipValue=1, samplerType=SAMPLERTYPE_Masks, texture=/Game/Athena/Environments/Meteor/Textures/Meteor_M}
  n11 = Multiply(A=VectorParameter() {parameterName=Chroma1Dark, defaultValue=<rgba 0.6 0 0.3876 0>}.RGB, B=@n10.B) {constB=0.25}
  n12 = TextureSample(UVs=Add(A=@n6, B=Multiply(A=@n7, B=@n8))) {constMipValue=1, samplerType=SAMPLERTYPE_Masks, texture=&Meteor_M}
  n13 = Multiply(A=@n12.B, B=VectorParameter() {parameterName=Chroma2Dark, defaultValue=<rgba 0 0.6 0.2984 0>}.RGB) {constB=0.2}
  n14 = LinearInterpolate(
   A=Multiply(A=VectorParameter() {parameterName=Chroma1, defaultValue=<rgba 0.9516 0 1 0>}.RGB, B=Add(A=@n10.G, B=@n11))
   B=Multiply(A=VectorParameter() {parameterName=Chroma2, defaultValue=<rgba 0 1 0.5733 0>}.RGB, B=Add(A=@n12.G, B=@n13))
  )
  n15 = Power(Base=@n14, Exp=ScalarParameter() {defaultValue=2, parameterName=GlowPower})
  n16 = Constant() {r=-2}
  Multiply(
   A=Add(A=VectorParameter() {parameterName=BackgroundColor, defaultValue=<rgba 0.0063 0.0818 0.2 0>}.RGB, B=Desaturation(None=@n15, Fraction=@n16))
   B=ScalarParameter() {defaultValue=8, parameterName=Brightness}
  )
  n17 = MaterialFunctionCall(Distance=ScalarParameter() {defaultValue=3000, parameterName=NearCamFadeDist}) {materialFunction=&Near_Camera_Fade}
  n18 = VertexInterpolator(VS=@n1.DistanceFadeMask)
  n19 = Fresnel(
   ExponentIn=LinearInterpolate(Alpha=ComponentMask(None=@n18.PS) {r=true}) {constA=1.5, constB=0.2}
   Normal=ComponentMask(None=SceneTexture() {sceneTextureId=PPI_WorldNormal}.Color) {b=true, g=true, r=true}
  ) {baseReflectFraction=0.1, exponent=1}
  n20 = Constant() {r=0.6}
  n21 = VectorParameter() {parameterName=PlaneLocation, defaultValue=<rgba 24056 2399 4844 0>}
  n22 = WorldPosition()
  n23 = DotProduct(
   A=Normalize(VectorInput=Subtract(A=@n21.RGB, B=@n22.XYZ))
   B=VectorParameter() {parameterName=MeteorDirection, defaultValue=<rgba 0.7755 -0.4565 -0.4361 0>}.RGB
  )
  n24 = DepthFade(
   Opacity=StaticSwitchParameter(True=Saturate(None=@n23), False=Constant() {r=1}) {parameterName=WorldSpacePortalMask}
   FadeDistance=ScalarParameter() {defaultValue=100, parameterName=DepthFadeDist}
  )
  n25 = Multiply(A=ShadingPathSwitch(Default=ShadingPathSwitch(Default=@n19, Forward=@n20), Mobile=@n20) {desc=Migrated from FeatureLevelSwitch}, B=@n24) {constA=1}
  n26 = ScalarParameter() {defaultValue=0.5, parameterName=Opacity}
  Multiply(A=Multiply(A=@n14, B=Multiply(A=@n17.Result, B=@n25)), B=@n26)
  n27 = ScalarParameter() {defaultValue=0.95, parameterName=Refraction}
  n28 = Multiply(A=@n26, B=@n24)
  ShadingPathSwitch(
   Default=MaterialFunctionCall!(RefractiveIndex=LinearInterpolate(B=@n27, Alpha=@n28) {constA=1}) {materialFunction=/Game/Effects/Fort_Effects/Materials/_Functions/MF_Dithered_Refraction}.Result
   Mobile=Constant() {r=1}
  ) {desc=Migrated from FeatureLevelSwitch}
 ]
}
