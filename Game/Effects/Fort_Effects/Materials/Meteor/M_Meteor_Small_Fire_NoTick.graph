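# Additive unlit fire material for small meteor effects, using a Sprite function with adjustable X/Y scale.
# Scaled vertex red manually drives an 8x8 fire flipbook and offsets a panning noise texture, while vertex blue toggles horizontal mirroring.
# The noise distorts flipbook UVs and, raised to a power, masks the two-color fire gradient.
# Flipbook alpha shapes the result, with depth fade skipped on mobile.
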Material /Game/Effects/Fort_Effects/Materials/Meteor/M_Meteor_Small_Fire_NoTick
graph {
 properties {
  allowTranslucentCustomDepthWrites=true
  blendMode=BLEND_Additive
  opacityMaskClipValue=0.01
  pixelDepthOffsetMode=PDOM_Legacy
  shadingModel=MSM_Unlit
  usedShadingModels=Unlit
 }
 nodes [
  MaterialFunctionCall!(
   Center Location=WorldPosition().XYZ
   0-1 UVs
   XY scale=AppendVector(A=ScalarParameter() {defaultValue=280, parameterName=Sprite Scale X}, B=ScalarParameter() {defaultValue=550, parameterName=Sprite Scale Y})
   OverridePivot
   "Normals (Optional)"
  ) {materialFunction=&Sprite}
  n1 = Constant() {r=8}
  n2 = Multiply(A=VertexColor().R, B=ScalarParameter() {defaultValue=5.3481, parameterName=Vertex Color Time Mult})
  n3 = TextureCoordinate()
  n4 = TextureCoordinate() {uTiling=-1}
  n5 = Floor(None=Add(A=VertexColor().B) {constB=0.055})
  n6 = TextureCoordinate() {uTiling=0.3333, vTiling=0.3333}
  n7 = Constant()
  n8 = ScalarParameter() {defaultValue=0.025, parameterName=Fire Pan Speed}
  n9 = TextureSample(UVs=Panner(Coordinate=Add(A=@n6, B=@n2), Speed=AppendVector(A=@n7, B=@n8)) {bFractionalPart=true}) {texture=&2ndInfectedMapLinear}
  n10 = Multiply(A=@n9.G, B=Constant2Vector() {g=0.25, r=1})
  n11 = MaterialFunctionCall!(
   U_Images=@n1
   V_Images=@n1
   SubUV Texture RGB=TextureObject() {texture=/Game/Effects/Fort_Effects/Textures/Fire/T_FireFluid_1_8X8}
   Time Speed
   Manual Time Control=StaticBool() {value=true}
   Time=@n2
   UVs=Add(A=LinearInterpolate(A=@n3, B=@n4, Alpha=@n5), B=Multiply(A=@n10) {constB=0.125})
  ) {materialFunction=/Game/Effects/Fort_Effects/Materials/_Functions/MF_SubUV}
  n12 = LinearInterpolate(
   A=VectorParameter() {parameterName=Color 1, defaultValue=<rgba 1 0 0 1>}.RGB
   B=VectorParameter() {parameterName=Color 2, defaultValue=<rgba 4 0.8612 0 1>}.RGB
   Alpha=@n11.RGB
  )
  n13 = DepthFade(FadeDistance=ScalarParameter() {defaultValue=200, parameterName=Depth Fade})
  n14 = Constant() {r=1}
  Multiply(
   A=Multiply(A=Multiply(A=@n12, B=@n11.A), B=ShadingPathSwitch(Default=@n13, Mobile=@n14))
   B=Power(Base=@n9.G, Exp=ScalarParameter() {defaultValue=4, parameterName=Pan Texture Power})
  )
 ]
}
