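# Masked unlit material for a meteor's smoke trail.
# Two panning noise layers drive smoke-pillar glow color, normal-based vertex deformation and an opacity mask shaped by gradients, camera distance and temporal dither.
# The WorldSpacePortalMask switch cuts the trail at a world plane facing MeteorDirection, collapsing vertices behind it onto PlaneLocation, adding a colored edge glow and scaling opacity by MPC_Snow's MeteorSmokeOpacity.
# All attributes pass through MF_CreativeMoveTool.
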Material /Game/Effects/Fort_Effects/Materials/Meteor/M_Meteor_Trail
graph {
 properties {
  bUseMaterialAttributes=true
  bUsedWithInstancedStaticMeshes=true
  blendMode=BLEND_Masked
  pixelDepthOffsetMode=PDOM_Legacy
  shadingModel=MSM_Unlit
  usedShadingModels=Unlit
 }
 nodes [
  ActorPositionWS()
  n1 = VectorParameter() {parameterName=PlaneLocation, defaultValue=<rgba 25049 1895 3704 0>}
  n2 = WorldPosition()
  n3 = DotProduct(
   A=Normalize(VectorInput=Subtract(A=@n1.RGB, B=@n2.XYZ))
   B=VectorParameter() {parameterName=MeteorDirection, defaultValue=<rgba 0.7755 -0.4565 -0.4361 0>}.RGB
  )
  n4 = Saturate(None=OneMinus(None=Abs(None=@n3)))
  n5 = ScalarParameter() {defaultValue=8, parameterName=PortalMaskPower}
  n6 = StaticSwitchParameter(
   True=Multiply(A=VectorParameter() {parameterName=PortalColor, defaultValue=<rgba 0.3125 14.647 20 0>}.RGB, B=Power(Base=@n4, Exp=@n5) {constExponent=4})
   False=Constant()
  ) {parameterName=WorldSpacePortalMask}
  n7 = WorldPosition()
  n8 = MaterialFunctionCall() {materialFunction=&ObjectPivotPoint}
  n9 = MaterialFunctionCall(Float2=TextureCoordinate() {vTiling=2}) {materialFunction=&BreakOutFloat2Components}
  n10 = OneMinus(None=MaterialFunctionCall() {materialFunction=&LinearGradient}.VGradient)
  n11 = ScalarParameter() {defaultValue=5, parameterName=Texture Scale Gradient Multiplier}
  n12 = Clamp(None=Power(Base=Multiply(A=@n10, B=@n11) {constB=5}, Exp=ScalarParameter() {defaultValue=2, parameterName=Texture Scale Gradient Power}))
  n13 = LinearInterpolate(A=ScalarParameter() {defaultValue=1, parameterName=ScaleU 2}, B=ScalarParameter() {defaultValue=1, parameterName=ScaleU}, Alpha=@n12)
  n14 = LinearInterpolate(A=ScalarParameter() {defaultValue=1, parameterName=ScaleV 2}, B=ScalarParameter() {defaultValue=1, parameterName=ScaleV}, Alpha=@n12)
  n15 = Multiply(A=@n9.G, B=@n14)
  n16 = ObjectPositionWS()
  n17 = Multiply(A=Divide(A=ComponentMask(None=@n16) {r=true}, B=ScalarParameter() {defaultValue=40000, parameterName=VerticalScale})) {constB=-1}
  n18 = ComponentMask(None=VectorParameter() {parameterName=Pan Speed, defaultValue=<rgba 0 -0.0065 0 1>}.RGB) {g=true, r=true}
  n19 = Panner(Coordinate=AppendVector(A=Multiply(A=@n9.R, B=@n13), B=Add(A=@n15, B=@n17)), Speed=@n18) {bFractionalPart=true, speedX=0.025, speedY=-1}
  n20 = TextureSample(UVs=@n19) {samplerType=SAMPLERTYPE_Normal, texture=/Game/Global/Textures/Default/Tileable_Patterns/2ndInfectedMapNormal}
  n21 = MaterialFunctionCall(Float2=TextureCoordinate() {uTiling=2, vTiling=4}) {materialFunction=&BreakOutFloat2Components}
  n22 = Multiply(A=@n21.G, B=@n14)
  n23 = Panner(Coordinate=AppendVector(A=Multiply(A=@n21.R, B=@n13), B=Add(A=@n17, B=@n22)), Speed=@n18) {bFractionalPart=true, speedX=-0.05, speedY=-1}
  n24 = TextureSample(UVs=@n23) {samplerType=SAMPLERTYPE_Normal, texture=&2ndInfectedMapNormal}
  n25 = Add(
   A=TextureSample(UVs=@n19) {texture=&2ndInfectedMapLinear}.G
   B=TextureSample(UVs=@n23) {texture=&2ndInfectedMapLinear}.G
  )
  n26 = MaterialFunctionCall!(
   GlowDirection=VectorParameter() {parameterName=MeteorDirection, defaultValue=<rgba 0.7755 -0.4565 -0.4361 0>}.RGB
   GlowStartHeight=ScalarParameter() {parameterName=GlowStartHeight}
   GlowWorldPosition=ComponentMask(None=Subtract(A=@n7.XYZ, B=@n8.Object Pivot Location)) {b=true}
   Normal=Transform(None=MaterialFunctionCall(BaseNormal=@n20.RGB, AdditionalNormal=@n24.RGB) {materialFunction=&BlendAngleCorrectedNormals}.Result)
   Mask=@n25
   Base Glow Mult
   Inner Glow Mult
   GlowDirectionBlend
  ) {materialFunction=/Game/Athena/Prototype/Blueprints/White/FX/MF_Giant_Smoke_Pillar_Color}
  n27 = Saturate(None=@n3)
  n28 = StaticSwitchParameter(
   True=Multiply(A=CollectionParameter() {collection=/Game/Athena/Prototype/Blueprints/White/MPC_Snow, parameterName=MeteorSmokeOpacity}, B=@n27)
   False=Constant() {r=1}
  ) {parameterName=WorldSpacePortalMask}
  n29 = MaterialFunctionCall(Alpha Threshold=ScalarParameter() {defaultValue=0.85, parameterName=Dither Threshold}) {materialFunction=&DitherTemporalAA}
  n30 = CameraPositionWS()
  n31 = WorldPosition()
  n32 = OneMinus(None=MaterialFunctionCall() {materialFunction=&LinearGradient}.VGradient)
  n33 = ScalarParameter() {defaultValue=5, group=Linear Gradient Falloff, parameterName=Linear Gradient Multiplier}
  n34 = Clamp(None=Power(
    Base=Multiply(A=@n32, B=@n33) {constB=5}
    Exp=ScalarParameter() {defaultValue=2, group=Linear Gradient Falloff, parameterName=Linear Gradient Power}
   ))
  n35 = Power(Base=@n25, Exp=ScalarParameter() {defaultValue=2, parameterName=Mask Texture Power})
  n36 = Power(
   Base=Multiply(
    A=MaterialFunctionCall() {materialFunction=&LinearGradient}.VGradient
    B=ScalarParameter() {defaultValue=5, group=Linear Gradient Falloff, parameterName=Linear Gradient Front Multiplier}
   ) {constB=5}
   Exp=ScalarParameter() {defaultValue=2, group=Linear Gradient Falloff, parameterName=Linear Gradient Front Power}
  )
  n37 = Power(
   Base=Multiply(A=Multiply(A=@n35, B=@n34), B=Clamp(None=@n36))
   Exp=ScalarParameter() {defaultValue=1, group=Linear Gradient Falloff, parameterName=Fade Gradient Power}
  )
  n38 = ScalarParameter() {defaultValue=4, group=Linear Gradient Falloff, parameterName=Fade Gradient Multiplier}
  n39 = Multiply(
   A=Divide(A=Distance(A=@n30, B=@n31.XYZ), B=ScalarParameter() {defaultValue=800, parameterName=Camera Distance Fade Falloff})
   B=Clamp(None=Multiply(A=@n37, B=@n38) {constB=4})
  )
  n40 = ShadingPathSwitch(Default=QualitySwitch(Default=Multiply(A=@n29.Result, B=@n39), Low=@n39), Mobile=@n39) {desc=Migrated from FeatureLevelSwitch}
  n41 = LinearInterpolate(Alpha=@n25) {constA=-1, constB=2}
  n42 = VertexNormalWS()
  n43 = VertexColor()
  n44 = ScalarParameter() {parameterName=Vert Color Offset}
  n45 = Multiply(A=Multiply(A=Multiply(A=@n41, B=@n42), B=ScalarParameter() {defaultValue=25, parameterName=Deform}), B=Saturate(None=Add(A=@n43.R, B=@n44)))
  n46 = LinearInterpolate(A=Subtract(A=@n1.RGB, B=WorldPosition().XYZ), B=@n45, Alpha=Ceil(None=@n27)) {constB=0}
  MaterialFunctionCall(MA In=MakeMaterialAttributes!(
    BaseColor
    Metallic
    Specular
    Roughness
    Anisotropy
    EmissiveColor=Add(A=@n6, B=@n26.Result)
    Opacity
    OpacityMask=Multiply(A=@n28, B=@n40)
    Normal
    Tangent
    WorldPositionOffset=StaticSwitchParameter(True=@n46, False=@n45) {parameterName=WorldSpacePortalMask}
    SubsurfaceColor
    ClearCoat
    ClearCoatRoughness
    AmbientOcclusion
    Refraction
    CustomizedUVs_0
    CustomizedUVs_1
    CustomizedUVs_2
    CustomizedUVs_3
    CustomizedUVs_4
    CustomizedUVs_5
    CustomizedUVs_6
    CustomizedUVs_7
    PixelDepthOffset
    ShadingModel
    Displacement
   )) {materialFunction=&MF_CreativeMoveTool}
 ]
}
