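# Translucent unlit material for a glitched meteor trail, combining a panning, chromatically split pattern with vertex distortion.
# Blended panning normals offset pattern samples in opposite directions, feeding two-color emissive glow and opacity.
# A world-space glitch mask function drives normal-based displacement and pushes along the meteor direction, added to pattern deformation scaled by vertex color.
# Opacity uses UV gradient depth fade, glitch mask and optional plane-based portal mask; refraction drops out on Low quality.
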Material /Game/Effects/Fort_Effects/Materials/Meteor/M_Meteor_Trail_Glitched
graph {
 properties {
  bUsedWithInstancedStaticMeshes=true
  bUsedWithNanite=true
  blendMode=BLEND_Translucent
  pixelDepthOffsetMode=PDOM_Legacy
  refractionMethod=RM_IndexOfRefraction
  shadingModel=MSM_Unlit
  usedShadingModels=Unlit
 }
 nodes [
  n1 = MaterialFunctionCall!(MaskScaleMult=ScalarParameter() {defaultValue=3, parameterName=WPOMaskScaleMult}) {materialFunction=/Game/Athena/Effects/Materials/Meteor/MF_Glitch_FX_World_Mask}
  n2 = Multiply(
   A=LinearInterpolate(B=LinearInterpolate(Alpha=@n1.DistanceFadeMask) {constA=50, constB=400}, Alpha=@n1.WorldSpaceMask) {constA=-5, constB=80}
   B=VertexNormalWS()
  )
  n3 = ScalarParameter() {defaultValue=1, parameterName=GlitchWPOMult}
  n4 = VectorParameter() {parameterName=MeteorDirection, defaultValue=<rgba 0.7755 -0.4565 -0.4361 0>}
  n5 = VertexNormalWS()
  n6 = Multiply(
   A=Multiply(A=@n1.WorldSpaceMask, B=LinearInterpolate(Alpha=@n1.DistanceFadeMask) {constA=0.5, constB=3})
   B=Multiply(A=@n4.RGB, B=DotProduct(A=@n4.RGB, B=@n5))
  )
  n7 = ScalarParameter() {defaultValue=1600, parameterName=GlitchPush}
  n8 = MaterialFunctionCall(Float2=TextureCoordinate() {vTiling=2}) {materialFunction=&BreakOutFloat2Components}
  n9 = OneMinus(None=MaterialFunctionCall() {materialFunction=&LinearGradient}.VGradient)
  n10 = ScalarParameter() {defaultValue=5, parameterName=Texture Scale Gradient Multiplier}
  n11 = Clamp(None=Power(Base=Multiply(A=@n9, B=@n10) {constB=5}, Exp=ScalarParameter() {defaultValue=0.2, parameterName=Texture Scale Gradient Power}))
  n12 = LinearInterpolate(A=ScalarParameter() {defaultValue=1, parameterName=ScaleU 2}, B=ScalarParameter() {defaultValue=1, parameterName=ScaleU}, Alpha=@n11)
  n13 = LinearInterpolate(A=ScalarParameter() {defaultValue=0.4, parameterName=ScaleV 2}, B=ScalarParameter() {defaultValue=0.4, parameterName=ScaleV}, Alpha=@n11)
  n14 = Multiply(A=@n8.G, B=@n13)
  n15 = ObjectPositionWS()
  n16 = Multiply(A=Divide(A=ComponentMask(None=@n15) {r=true}, B=ScalarParameter() {defaultValue=40000, parameterName=VerticalScale})) {constB=-1}
  n17 = ComponentMask(None=VectorParameter() {parameterName=Pan Speed, defaultValue=<rgba 0 -0.0065 0 1>}.RGB) {g=true, r=true}
  n18 = Panner(Coordinate=AppendVector(A=Multiply(A=@n12, B=@n8.R), B=Add(A=@n14, B=@n16)), Speed=@n17) {bFractionalPart=true, speedX=0.025, speedY=-1}
  n19 = TextureSample(UVs=@n18) {samplerType=SAMPLERTYPE_Normal, texture=/Game/Global/Textures/Default/Tileable_Patterns/2ndInfectedMapNormal}
  n20 = MaterialFunctionCall(Float2=TextureCoordinate() {uTiling=2, vTiling=4}) {materialFunction=&BreakOutFloat2Components}
  n21 = Multiply(A=@n20.G, B=@n13)
  n22 = Panner(Coordinate=AppendVector(A=Multiply(A=@n12, B=@n20.R), B=Add(A=@n16, B=@n21)), Speed=@n17) {bFractionalPart=true, speedX=-0.05, speedY=-1}
  n23 = TextureSample(UVs=@n22) {samplerType=SAMPLERTYPE_Normal, texture=&2ndInfectedMapNormal}
  n24 = ComponentMask(None=Transform(None=MaterialFunctionCall(BaseNormal=@n19.RGB, AdditionalNormal=@n23.RGB) {materialFunction=&BlendAngleCorrectedNormals}.Result) {transformType=TRANSFORM_View}) {g=true, r=true}
  n25 = ScalarParameter() {defaultValue=0.01, parameterName=ChromaOffset}
  n26 = Multiply(A=@n25) {constB=-1}
  n27 = TextureSample(UVs=Add(A=Multiply(A=@n24, B=@n26), B=@n18)) {texture=&2ndInfectedMapLinear}
  n28 = TextureSample(UVs=@n22) {texture=&2ndInfectedMapLinear}
  n29 = Multiply(A=VertexNormalWS(), B=LinearInterpolate(Alpha=Add(A=@n27.G, B=@n28.G)) {constA=-0.7, constB=2})
  n30 = ScalarParameter() {defaultValue=580, parameterName=Deform}
  n31 = Add(A=VertexColor().R, B=ScalarParameter() {defaultValue=0.175, parameterName=Vert Color Offset})
  Add(A=Add(A=Multiply(A=@n2, B=@n3), B=Multiply(A=@n6, B=@n7)), B=Multiply(A=Multiply(A=@n29, B=@n30), B=Saturate(None=@n31)))
  n32 = VectorParameter() {parameterName=BackgroundColor, defaultValue=<rgba 0.0063 0.0818 0.2 0>}
  n33 = Multiply(A=@n27.G, B=@n27.G)
  n34 = Multiply(A=VectorParameter() {parameterName=Chroma1, defaultValue=<rgba 0.9516 0 1 0>}.RGB, B=@n33)
  n35 = TextureSample(UVs=Add(A=Multiply(A=@n24, B=@n25), B=@n18)) {texture=&2ndInfectedMapLinear}
  n36 = Multiply(A=@n35.G, B=@n35.G)
  n37 = Multiply(A=VectorParameter() {parameterName=Chroma2, defaultValue=<rgba 0 1 0.5733 0>}.RGB, B=@n36)
  n38 = Desaturation(
   None=Power(Base=LinearInterpolate(A=@n34, B=@n37), Exp=ScalarParameter() {defaultValue=2, parameterName=GlowPower})
   Fraction=Constant() {r=-2}
  )
  n39 = OneMinus(None=ComponentMask(None=TextureCoordinate()) {g=true})
  n40 = Multiply(A=@n39) {constB=3}
  Multiply(A=Multiply(A=Add(A=@n32.RGB, B=@n38), B=Add(A=@n40) {constB=0.8}), B=ScalarParameter() {defaultValue=7, parameterName=Brightness})
  n41 = VectorParameter() {parameterName=PlaneLocation, defaultValue=<rgba 24056 2399 4844 0>}
  n42 = WorldPosition()
  n43 = DotProduct(
   A=Normalize(VectorInput=Subtract(A=@n41.RGB, B=@n42.XYZ))
   B=VectorParameter() {parameterName=MeteorDirection, defaultValue=<rgba 0.7755 -0.4565 -0.4361 0>}.RGB
  )
  n44 = ScalarParameter() {defaultValue=1000, parameterName=DepthFadeDist}
  n45 = VertexInterpolator(VS=AppendVector(A=@n1.WorldSpaceMask, B=@n1.WorldSpaceMask))
  n46 = Multiply(
   A=StaticSwitchParameter(True=Saturate(None=@n43), False=Constant() {r=1}) {parameterName=WorldSpacePortalMask}
   B=Multiply(A=DepthFade(Opacity=@n39, FadeDistance=@n44), B=ComponentMask(None=@n45.PS) {r=true})
  )
  n47 = ScalarParameter() {defaultValue=1, parameterName=Opacity}
  Multiply(A=Multiply(A=LinearInterpolate(A=@n33, B=@n36), B=@n46), B=@n47)
  n48 = Multiply(A=@n46, B=Multiply(A=@n1.WorldSpaceMask, B=@n1.WorldSpaceMask))
  QualitySwitch(
   Default=LinearInterpolate(B=ScalarParameter() {defaultValue=0.95, parameterName=Refraction}, Alpha=Multiply(A=@n47, B=@n48)) {constA=1}
   Low=Constant() {r=1}
  )
 ]
}
